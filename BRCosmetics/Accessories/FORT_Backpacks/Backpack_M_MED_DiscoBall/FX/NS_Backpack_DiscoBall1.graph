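# Niagara effect for the disco ball backpack cosmetic that makes pink, purple and blue sparkles twinkle on the ball's surface.
# Two spawn groups feed separate hemisphere shells, re-centered every update on the dyn_disco bone and counter-rotated by system age around bone-relative axes.
# Curves swell and fade each sprite over its life, while time-of-day functions dim and tint color and glow for two sprite renderers.
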NiagaraSystem /BRCosmetics/Accessories/FORT_Backpacks/Backpack_M_MED_DiscoBall/FX/NS_Backpack_DiscoBall1
dependencies {
 usedDataInterfaces [&NiagaraDataInterfaceSkeletalMesh, &NiagaraDataInterfaceCurve, &NiagaraDataInterfaceColorCurve]
 usedDynamicInputs []
 usedModules [&SystemState]
}
emitters {SurfaceSparkles {
  emitterUpdateScript [
   &EmitterState()
   &SpawnRate(SpawnRate=6, Use Spawn Probability=false) {module=SpawnRate001}
   &SpawnRate(SpawnGroup=1, SpawnRate=5, Use Spawn Probability=false)
  ]
  particleSpawnScript [
   &InitializeParticle(
    Color=&MaskLinearColorBySpawnGroup(
     Masked Linear Color=&RandomRangeLinearColor(Minimum=<rgba 1.2662 0 2 1>, Maximum=<rgba 2 0.4036 1.9362 1>)
     Passthrough Linear Color=<rgba 0.2378 0 1 1>
    )
    Lifetime Mode=::NewEnumerator1
    Sprite Size Mode=::NewEnumerator1
    Sprite UV Mode=::NewEnumerator3
    Uniform Sprite Size Max=20
    Uniform Sprite Size Min=10
    UsePositionOffset=false
    Write Material Random=true
   )
   &CameraOffset(Camera Offset Amount=15)
  ]
  particleUpdateScript [
   /SetVariables(Particles.GlowColor=&MultiplyLinearColor(
     Linear Color=&MaskLinearColorBySpawnGroup(
      Masked Linear Color=&RandomRangeLinearColor(Minimum=<rgba 0.6331 0 1 1>, Maximum=<rgba 1 0.2228 0.9669 1>)
      Passthrough Linear Color=<rgba 0.2184 0 1 1>
     )
     Scale Factor=&MakeVector4(
      X=&FNTimeOfDayScalar(Night Scale=0.5, Morning Scale=1, Evening Scale=0.5)
      Y=&FNTimeOfDayScalar(Night Scale=0.5, Morning Scale=1, Evening Scale=0.5)
      Z=&FNTimeOfDayScalar(Night Scale=0.5, Morning Scale=1, Evening Scale=0.5)
      W=&FNTimeOfDayScalar(Night Scale=0.5, Morning Scale=1, Evening Scale=0.5)
     )
    )) {isSetParameters=true}
   &SkeletalMeshLocation(
    Bone Sampling Mode=::NewEnumerator4
    MaskSpawnGroup=false
    Position Sampling=::NewEnumerator1
    Skeletal Mesh=DataInterface() {propertyValues {
      DefaultMesh=null
      ExcludeBoneName=root
      OptionalComponentClass=null
      PreviewMesh=/Game/Accessories/FORT_Backpacks/Backpack_M_MED_DiscoBall/Meshes/M_MED_DiscoBall_Pack
      SkinningMode=SkinOnTheFly
      SoftSourceActor=null
      SourceMode=Default
      UvSetIndex=0
      WholeMeshLOD=-1
      bExcludeBone=false
      bReadDeformedGeometry=true
      bRequireCurrentFrameData=true
      ComponentTags []
      FilteredBones [dyn_disco]
      FilteredSockets []
      MeshUserParameter {parameter {name=None, typeDefHandle {registeredTypeIndex=1}, varData []}}
      SamplingRegions []
     }}
   )
   &ShapeLocation(
    Hemisphere Distribution=<1 0.6>
    MaskSpawnGroup=true
    Override Seed=false
    Rotation Angle:NiagaraFloat=&Multiply_Float(A=$Engine.System.Age, B=0.035)
    Rotation Axis=&RotateVectorByQuaternion(Vector To Rotate=<-1 0 0>, Quaternion=$Particles.SkeletalMeshLocation.SampledRotation)
    Rotation Mode=::NewEnumerator1
    Shape Origin=$Particles.SkeletalMeshLocation.SampledPosition
    Sphere Radius=11
    Sphere Surface Distribution=1
    UseOwnerScale=true
   )
   &ShapeLocation(
    Hemisphere Distribution=<1 0.45>
    MaskSpawnGroup=true
    Override Seed=false
    Rotation Angle:NiagaraFloat=&Multiply_Float(A=$Engine.System.Age, B=-0.035)
    Rotation Axis=&RotateVectorByQuaternion(Vector To Rotate=<1 0 0>, Quaternion=$Particles.SkeletalMeshLocation.SampledRotation)
    Rotation Mode=::NewEnumerator1
    Shape Origin=$Particles.SkeletalMeshLocation.SampledPosition
    Spawn Group Mask=1
    Sphere Radius=11
    Sphere Surface Distribution=1
    UseOwnerScale=true
   ) {module=ShapeLocation001}
   &ParticleState()
   &ScaleSpriteSize(Scale Factor=&Vector2DFromFloat(Value=&FloatFromCurve(FloatCurve=DataInterface() {propertyValues {Curve {keys [
          {interpMode=RCIM_Cubic, tangentMode=RCTM_User, time=0, value=0}
          {interpMode=RCIM_Cubic, tangentMode=RCTM_User, time=0.5, value=1}
          {tangentMode=RCTM_User, time=1, value=0}
         ]}}})))
   &ScaleColor(
    Scale Alpha=&FloatFromCurve(FloatCurve=DataInterface() {propertyValues {Curve {keys [
         {interpMode=RCIM_Cubic, tangentMode=RCTM_User, time=0, value=0}
         {interpMode=RCIM_Cubic, tangentMode=RCTM_User, time=0.5, value=1}
         {tangentMode=RCTM_User, time=1, value=0}
        ]}}})
    Scale RGB=&FN_Select_Vector_from_Time_Of_Day(Morning=<1 1 1>, Day=<1 1 1>, Evening=<0.3 0 0.6>, Night=<0.3 0 0.6>)
    ScaleA=true
    ScaleRGB=true
   )
   &SolveForcesAndVelocity(Clamp Velocity=false, Limit Acceleration=false)
  ]
  rendererClasses [&NiagaraSpriteRendererProperties]
  renderers [
   {rendererClass=&NiagaraSpriteRendererProperties}
   {rendererClass=&NiagaraSpriteRendererProperties, rendererIndex=1}
  ]
 }}
summary {
 systemName=NS_Backpack_DiscoBall1
 emitters [{emitterName=SurfaceSparkles, rendererClasses [&NiagaraSpriteRendererProperties]}]
 userVariables []
}
